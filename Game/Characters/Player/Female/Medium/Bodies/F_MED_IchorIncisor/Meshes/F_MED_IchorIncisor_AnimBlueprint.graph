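# Animation Blueprint for the IchorIncisor female player body, adding secondary motion to its jacket, tie and sleeves.
# The parent mesh pose is copied through a Control Rig, then chained AnimDynamics and additive bone offsets, with spherical and planar limits for collision.
# Back and side jacket physics switch off when a backpack is worn.
# The event graph also drives two-region material WPO cloth sim from hand and foot bones, plus wind during skydive floating.
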AnimBlueprint /Game/Characters/Player/Female/Medium/Bodies/F_MED_IchorIncisor/Meshes/F_MED_IchorIncisor_AnimBlueprint
target_skeleton=&/F_MED_IchorIncisor_Skeleton
graphs [
 AnimGraph {nodes [
   n1 = Control Rig(Source=Use cached pose 'SavedPose26139'()) {
    alphaBoolBlend {blendOption=Linear}
    controlRigAssetReference {blueprintRigClass=&/F_MED_IchorIncisor_CtrlRig.F_MED_IchorIncisor_CtrlRig_C}
   }
   n2 = "Anim Dynamics - Chain: dyn_jacket_fr_side_l_01 -> dyn_jacket_fr_side_l_02"(ComponentPose=Local To Component(LocalPose=@n1), Alpha=1, ExternalForce="0,0,0") {
    angularDampingOverride=0.999
    angularSpringConstant=300
    bAngularSpring=true
    bChain=true
    bOverrideAngularDamping=true
    bUseGravityOverride=true
    bUseSphericalLimits=true
    simulationSpace=BoneRelative
    alphaBoolBlend {blendOption=Linear}
    boundBone {boneName=dyn_jacket_fr_side_l_01}
    chainEnd {boneName=dyn_jacket_fr_side_l_02}
    componentAppliedLinearAccClamp=<2000 2000 2000>
    componentLinearAccScale=<0.4 0.4 0.4>
    componentLinearVelScale=<0.4 0.4 0.4>
    gravityOverride {z=-980}
    physicsBodyDefinitions [
     {
      boundBone {boneName=dyn_jacket_fr_side_l_01}
      boxExtents=<30 30 30>
      constraintSetup {angularLimitsMax=<0 5 20>, angularLimitsMin=<0 -5 -2>, angularTarget=<0 0 0>}
      localJointOffset=<-1 0 0>
     }
     {
      boundBone {boneName=dyn_jacket_fr_side_l_02}
      boxExtents=<30 30 30>
      constraintSetup {angularLimitsMax=<0 5 30>, angularLimitsMin=<0 -5 -5>, angularTarget=<0 0 0>}
      localJointOffset=<-1 0 0>
     }
    ]
    relativeSpaceBone {boneName=root}
    sphericalLimits [
     {limitRadius=11, drivingBone {boneName=spine_02}, sphereLocalOffset=<0 0 0>}
     {limitRadius=12, drivingBone {boneName=spine_04}, sphereLocalOffset=<0 0 0>}
    ]
   }
   n3 = "Anim Dynamics - Chain: dyn_jacket_side_l_01 -> dyn_jacket_side_l_02"(ComponentPose=@n2, Alpha=Get Alpha_1() -> (Alpha_1:float), ExternalForce="0,0,0") {
    angularDampingOverride=0.999
    angularSpringConstant=300
    bAngularSpring=true
    bChain=true
    bOverrideAngularDamping=true
    bUseGravityOverride=true
    bUseSphericalLimits=true
    simulationSpace=BoneRelative
    alphaBoolBlend {blendOption=Linear}
    boundBone {boneName=dyn_jacket_side_l_01}
    chainEnd {boneName=dyn_jacket_side_l_02}
    componentAppliedLinearAccClamp=<2000 2000 2000>
    componentLinearAccScale=<0.4 0.4 0.4>
    componentLinearVelScale=<0.4 0.4 0.4>
    gravityOverride {z=-980}
    physicsBodyDefinitions [
     {
      boundBone {boneName=dyn_jacket_side_l_01}
      boxExtents=<30 30 30>
      constraintSetup {angularLimitsMax=<0 5 10>, angularLimitsMin=<0 -5 -1>, angularTarget=<0 0 0>}
      localJointOffset=<-1 0 0>
     }
     {
      boundBone {boneName=dyn_jacket_side_l_02}
      boxExtents=<30 30 30>
      constraintSetup {angularLimitsMax=<0 5 10>, angularLimitsMin=<0 -5 -5>, angularTarget=<0 0 0>}
      localJointOffset=<-1 0 0>
     }
    ]
    relativeSpaceBone {boneName=root}
    sphericalLimits [
     {limitRadius=11, drivingBone {boneName=spine_02}, sphereLocalOffset=<0 0 0>}
     {limitRadius=12, drivingBone {boneName=spine_04}, sphereLocalOffset=<0 0 0>}
    ]
   }
   n4 = "Anim Dynamics - Chain: dyn_jacket_back_01 -> dyn_jacket_back_02"(ComponentPose=@n3, Alpha=Get Alpha() -> (Alpha:float), ExternalForce="0,0,0") {
    angularDampingOverride=0.999
    angularSpringConstant=300
    bAngularSpring=true
    bChain=true
    bOverrideAngularDamping=true
    bUseGravityOverride=true
    bUseSphericalLimits=true
    simulationSpace=BoneRelative
    alphaBoolBlend {blendOption=Linear}
    boundBone {boneName=dyn_jacket_back_01}
    chainEnd {boneName=dyn_jacket_back_02}
    componentAppliedLinearAccClamp=<2000 2000 2000>
    componentLinearAccScale=<0.4 0.4 0.4>
    componentLinearVelScale=<0.4 0.4 0.4>
    gravityOverride {z=-980}
    physicsBodyDefinitions [
     {
      boundBone {boneName=dyn_jacket_back_01}
      boxExtents=<30 30 30>
      constraintSetup {angularLimitsMax=<0 5 10>, angularLimitsMin=<0 -5 0>, angularTarget=<0 0 0>}
      localJointOffset=<-1 0 0>
     }
     {
      boundBone {boneName=dyn_jacket_back_02}
      boxExtents=<30 30 30>
      constraintSetup {angularLimitsMax=<0 5 10>, angularLimitsMin=<0 -5 -10>, angularTarget=<0 0 0>}
      localJointOffset=<-1 0 0>
     }
    ]
    relativeSpaceBone {boneName=root}
    sphericalLimits [
     {limitRadius=10.383, drivingBone {boneName=spine_02}, sphereLocalOffset=<0 0 0>}
     {limitRadius=8.82, drivingBone {boneName=spine_04}, sphereLocalOffset=<0 0.7814 0>}
     {limitRadius=3, drivingBone {boneName=dyn_jacket_fr_l_03}, sphereLocalOffset=<1.1208 -2.7937 -0.4248>}
     {limitRadius=3, drivingBone {boneName=dyn_jacket_fr_r_03}, sphereLocalOffset=<-1.2184 -0.1441 0.5445>}
     {limitRadius=4, drivingBone {boneName=dyn_jacket_fr_l_02}, sphereLocalOffset=<1.1036 -2.2176 -0.1647>}
     {limitRadius=3, drivingBone {boneName=dyn_jacket_fr_r_02}, sphereLocalOffset=<-1.2184 -0.2916 0.5475>}
     {limitRadius=4, drivingBone {boneName=dyn_jacket_fr_l_01}, sphereLocalOffset=<1.8541 -2.2239 0.0236>}
    ]
   }
   n5 = "Anim Dynamics - Chain: dyn_jacket_fr_side_r_01 -> dyn_jacket_fr_side_r_02"(ComponentPose=@n4, Alpha=1, ExternalForce="0,0,0") {
    angularDampingOverride=0.999
    angularSpringConstant=300
    bAngularSpring=true
    bChain=true
    bOverrideAngularDamping=true
    bUseGravityOverride=true
    bUseSphericalLimits=true
    simulationSpace=BoneRelative
    alphaBoolBlend {blendOption=Linear}
    boundBone {boneName=dyn_jacket_fr_side_r_01}
    chainEnd {boneName=dyn_jacket_fr_side_r_02}
    componentAppliedLinearAccClamp=<2000 2000 2000>
    componentLinearAccScale=<0.4 0.4 0.4>
    componentLinearVelScale=<0.4 0.4 0.4>
    gravityOverride {z=-980}
    physicsBodyDefinitions [
     {
      boundBone {boneName=dyn_jacket_fr_side_r_01}
      boxExtents=<30 30 30>
      constraintSetup {angularLimitsMax=<0 1 20>, angularLimitsMin=<0 -1 0>, angularTarget=<0 0 0>}
      localJointOffset=<1 0 0>
     }
     {
      boundBone {boneName=dyn_jacket_fr_side_r_02}
      boxExtents=<30 30 30>
      constraintSetup {angularLimitsMax=<0 5 30>, angularLimitsMin=<0 -5 0>, angularTarget=<0 0 0>}
      localJointOffset=<1 0 0>
     }
    ]
    relativeSpaceBone {boneName=root}
    sphericalLimits [
     {limitRadius=12.11, drivingBone {boneName=spine_02}, sphereLocalOffset=<0 0 0>}
     {limitRadius=12.3277, drivingBone {boneName=spine_03}, sphereLocalOffset=<0.2427 0 0>}
     {limitRadius=12, drivingBone {boneName=spine_04}, sphereLocalOffset=<0 0 0>}
     {limitRadius=4, drivingBone {boneName=dyn_jacket_fr_l_01}, sphereLocalOffset=<1.8541 -2.2239 0.0236>}
    ]
   }
   n6 = "Anim Dynamics - Chain: dyn_jacket_side_r_01 -> dyn_jacket_side_r_02"(ComponentPose=@n5, Alpha=Get Alpha_0() -> (Alpha_0:float), ExternalForce="0,0,0") {
    angularDampingOverride=0.999
    angularSpringConstant=300
    bAngularSpring=true
    bChain=true
    bOverrideAngularDamping=true
    bUseGravityOverride=true
    bUseSphericalLimits=true
    simulationSpace=BoneRelative
    alphaBoolBlend {blendOption=Linear}
    boundBone {boneName=dyn_jacket_side_r_01}
    chainEnd {boneName=dyn_jacket_side_r_02}
    componentAppliedLinearAccClamp=<2000 2000 2000>
    componentLinearAccScale=<0.4 0.4 0.4>
    componentLinearVelScale=<0.4 0.4 0.4>
    gravityOverride {z=-980}
    physicsBodyDefinitions [
     {
      boundBone {boneName=dyn_jacket_side_r_01}
      boxExtents=<30 30 30>
      constraintSetup {angularLimitsMax=<0 5 10>, angularLimitsMin=<0 -5 -1>, angularTarget=<0 0 0>}
      localJointOffset=<-1 0 0>
     }
     {
      boundBone {boneName=dyn_jacket_side_r_02}
      boxExtents=<30 30 30>
      constraintSetup {angularLimitsMax=<0 5 10>, angularLimitsMin=<0 -5 -1>, angularTarget=<0 0 0>}
      localJointOffset=<-1 0 0>
     }
    ]
    relativeSpaceBone {boneName=root}
    sphericalLimits [
     {limitRadius=11, drivingBone {boneName=spine_02}, sphereLocalOffset=<0 0 0>}
     {limitRadius=12, drivingBone {boneName=spine_03}, sphereLocalOffset=<0 0 0>}
    ]
   }
   n7 = "Transform (Modify) Bone - Bone: dyn_jacket_side_l_MSTR"!(ComponentPose=@n6, bAlphaBoolEnabled:bool=True, Translation="0,0,0", Rotation="0,-3,0", Scale="1,1,1") -> (Pose) {
    alphaInputType=Bool
    rotationMode=BMM_Additive
    rotationSpace=BCS_BoneSpace
    translationMode=BMM_Additive
    translationSpace=BCS_BoneSpace
    alphaBoolBlend {blendInTime=0.5, blendOption=CubicInOut, blendOutTime=0.5}
    alphaScaleBiasClamp {bInterpResult=true, interpSpeedDecreasing=2, interpSpeedIncreasing=2}
    boneToModify {boneName=dyn_jacket_side_l_MSTR}
    rotation {yaw=-3}
   }
   n8 = "Transform (Modify) Bone - Bone: dyn_jacket_back_MSTR"!(ComponentPose=@n7, bAlphaBoolEnabled:bool=True, Translation="0,0,0", Rotation="0,-3,0", Scale="1,1,1") -> (Pose) {
    alphaInputType=Bool
    rotationMode=BMM_Additive
    rotationSpace=BCS_BoneSpace
    translationMode=BMM_Additive
    translationSpace=BCS_BoneSpace
    alphaBoolBlend {blendInTime=0.5, blendOption=CubicInOut, blendOutTime=0.5}
    alphaScaleBiasClamp {bInterpResult=true, interpSpeedDecreasing=2, interpSpeedIncreasing=2}
    boneToModify {boneName=dyn_jacket_back_MSTR}
    rotation {yaw=-3}
   }
   n9 = "Transform (Modify) Bone - Bone: dyn_jacket_side_r_MSTR"!(ComponentPose=@n8, bAlphaBoolEnabled:bool=True, Translation="0,0,0", Rotation="0,-3,0", Scale="1,1,1") -> (Pose) {
    alphaInputType=Bool
    rotationMode=BMM_Additive
    rotationSpace=BCS_BoneSpace
    translationMode=BMM_Additive
    translationSpace=BCS_BoneSpace
    alphaBoolBlend {blendInTime=0.5, blendOption=CubicInOut, blendOutTime=0.5}
    alphaScaleBiasClamp {bInterpResult=true, interpSpeedDecreasing=2, interpSpeedIncreasing=2}
    boneToModify {boneName=dyn_jacket_side_r_MSTR}
    rotation {yaw=-3}
   }
   n10 = "Anim Dynamics - Chain: dyn_jacket_fr_r_01 -> dyn_jacket_fr_r_03"(ComponentPose=@n9, Alpha=1, ExternalForce="0,0,0") {
    angularDampingOverride=0.999
    angularSpringConstant=300
    bAngularSpring=true
    bChain=true
    bOverrideAngularDamping=true
    bUseGravityOverride=true
    bUseSphericalLimits=true
    simulationSpace=BoneRelative
    alphaBoolBlend {blendOption=Linear}
    boundBone {boneName=dyn_jacket_fr_r_01}
    chainEnd {boneName=dyn_jacket_fr_r_03}
    componentAppliedLinearAccClamp=<2000 2000 2000>
    componentLinearAccScale=<0.4 0.4 0.4>
    componentLinearVelScale=<0.4 0.4 0.4>
    gravityOverride {z=-980}
    physicsBodyDefinitions [
     {
      boundBone {boneName=dyn_jacket_fr_r_01}
      boxExtents=<30 30 30>
      constraintSetup {angularLimitsMax=<0 1 10>, angularLimitsMin=<0 -1 0>, angularTarget=<0 0 0>}
      localJointOffset=<1 0 0>
     }
     {
      boundBone {boneName=dyn_jacket_fr_r_02}
      boxExtents=<30 30 30>
      constraintSetup {angularLimitsMax=<0 5 10>, angularLimitsMin=<0 -5 0>, angularTarget=<0 0 0>}
      localJointOffset=<1 0 0>
     }
     {
      boundBone {boneName=dyn_jacket_fr_r_03}
      boxExtents=<30 30 30>
      constraintSetup {angularLimitsMax=<0 8 15>, angularLimitsMin=<0 -8 0>, angularTarget=<0 0 0>}
      localJointOffset=<1 0 0>
     }
    ]
    relativeSpaceBone {boneName=root}
    sphericalLimits [
     {limitRadius=13.8084, drivingBone {boneName=pelvis}, sphereLocalOffset=<1.457 0 0>}
     {limitRadius=11.8399, drivingBone {boneName=spine_01}, sphereLocalOffset=<4.0761 0 0>}
     {limitRadius=12, drivingBone {boneName=spine_04}, sphereLocalOffset=<0 0 0>}
     {limitRadius=4.4118, drivingBone {boneName=dyn_tie_03}, sphereLocalOffset=<1.9081 -1.78 -0.0023>}
     {limitRadius=12, drivingBone {boneName=spine_03}, sphereLocalOffset=<0 0 0>}
    ]
   }
   n11 = "Anim Dynamics - Chain: dyn_jacket_fr_l_01 -> dyn_jacket_fr_l_03"(ComponentPose=@n10, Alpha=1, ExternalForce="0,0,0") {
    angularDampingOverride=0.999
    angularSpringConstant=300
    bAngularSpring=true
    bChain=true
    bOverrideAngularDamping=true
    bUseGravityOverride=true
    bUseSphericalLimits=true
    simulationSpace=BoneRelative
    alphaBoolBlend {blendOption=Linear}
    boundBone {boneName=dyn_jacket_fr_l_01}
    chainEnd {boneName=dyn_jacket_fr_l_03}
    componentAppliedLinearAccClamp=<2000 2000 2000>
    componentLinearAccScale=<0.4 0.4 0.4>
    componentLinearVelScale=<0.4 0.4 0.4>
    gravityOverride {z=-980}
    physicsBodyDefinitions [
     {
      boundBone {boneName=dyn_jacket_fr_l_01}
      boxExtents=<30 30 30>
      constraintSetup {angularLimitsMax=<0 5 10>, angularLimitsMin=<0 -5 0>, angularTarget=<0 0 0>}
      localJointOffset=<-1 0 0>
     }
     {
      boundBone {boneName=dyn_jacket_fr_l_02}
      boxExtents=<30 30 30>
      constraintSetup {angularLimitsMax=<0 5 10>, angularLimitsMin=<0 -5 0>, angularTarget=<0 0 0>}
      localJointOffset=<-1 0 0>
     }
     {
      boundBone {boneName=dyn_jacket_fr_l_03}
      boxExtents=<30 30 30>
      constraintSetup {angularLimitsMax=<0 8 15>, angularLimitsMin=<0 -8 0>, angularTarget=<0 0 0>}
      localJointOffset=<-1 0 0>
     }
    ]
    relativeSpaceBone {boneName=root}
    sphericalLimits [
     {limitRadius=11, drivingBone {boneName=spine_01}, sphereLocalOffset=<4.0761 0 0>}
     {limitRadius=13.2842, drivingBone {boneName=pelvis}, sphereLocalOffset=<4.0761 0 0>}
     {limitRadius=12, drivingBone {boneName=spine_04}, sphereLocalOffset=<0 0 0>}
     {limitRadius=4, drivingBone {boneName=dyn_tie_03}, sphereLocalOffset=<1.9081 0 -0.0004>}
     {limitRadius=4, drivingBone {boneName=dyn_tie_02}, sphereLocalOffset=<3.2475 0 -0.0672>}
     {limitRadius=12, drivingBone {boneName=spine_03}, sphereLocalOffset=<0 0 0>}
    ]
   }
   n12 = "Anim Dynamics - Chain: dyn_tie_01 -> dyn_tie_03"(ComponentPose=@n11, Alpha=1, ExternalForce="0,0,0") {
    angularDampingOverride=0.999
    angularSpringConstant=300
    bAngularSpring=true
    bChain=true
    bOverrideAngularDamping=true
    bUseGravityOverride=true
    bUseSphericalLimits=true
    simulationSpace=BoneRelative
    alphaBoolBlend {blendOption=Linear}
    boundBone {boneName=dyn_tie_01}
    chainEnd {boneName=dyn_tie_03}
    componentAppliedLinearAccClamp=<2000 2000 2000>
    componentLinearAccScale=<0.4 0.4 0.4>
    componentLinearVelScale=<0.4 0.4 0.4>
    gravityOverride {z=-980}
    physicsBodyDefinitions [
     {
      boundBone {boneName=dyn_tie_01}
      boxExtents=<50 50 50>
      constraintSetup {angularLimitsMax=<0 0 5>, angularLimitsMin=<0 0 -5>, angularTarget=<0 0 0>}
      localJointOffset=<1 0 0>
     }
     {
      boundBone {boneName=dyn_tie_02}
      boxExtents=<50 50 50>
      constraintSetup {angularLimitsMax=<0 0 10>, angularLimitsMin=<0 0 -10>, angularTarget=<0 0 0>}
      localJointOffset=<1 0 0>
     }
     {
      boundBone {boneName=dyn_tie_03}
      boxExtents=<30 30 30>
      constraintSetup {angularLimitsMax=<0 7 15>, angularLimitsMin=<0 -7 -15>, angularTarget=<0 0 0>}
      localJointOffset=<1 0 0>
     }
    ]
    planarLimits [
     {drivingBone {boneName=dyn_jacket_fr_l_02}, planeTransform {location=<0 0 0>, rotation=<rot -0.0005 -0.0002 -180>}}
     {drivingBone {boneName=dyn_jacket_fr_l_03}, planeTransform {location=<0 0 0>, rotation=<rot -0.0005 -0.0002 -180>}}
    ]
    relativeSpaceBone {boneName=root}
    sphericalLimits [
     {limitRadius=11, drivingBone {boneName=spine_02}, sphereLocalOffset=<0 0 0>}
     {limitRadius=12, drivingBone {boneName=spine_04}, sphereLocalOffset=<0 0 0>}
     {limitRadius=3, drivingBone {boneName=dyn_jacket_fr_l_03}, sphereLocalOffset=<1.1208 -2.7937 -0.4248>}
     {limitRadius=3, drivingBone {boneName=dyn_jacket_fr_r_03}, sphereLocalOffset=<-1.2184 -0.1441 0.5445>}
     {limitRadius=4, drivingBone {boneName=dyn_jacket_fr_l_02}, sphereLocalOffset=<1.1036 -2.2176 -0.1647>}
     {limitRadius=3, drivingBone {boneName=dyn_jacket_fr_r_02}, sphereLocalOffset=<-1.2184 -0.2916 0.5475>}
     {limitRadius=4, drivingBone {boneName=dyn_jacket_fr_l_01}, sphereLocalOffset=<1.8541 -2.2239 0.0236>}
    ]
   }
   n13 = "Transform (Modify) Bone - Bone: dyn_sleeves_r_bk"(ComponentPose=@n12, Alpha=1, Translation="-2,0,0", Rotation="-10,15,-10", Scale="1,1,1") {
    rotationMode=BMM_Additive
    rotationSpace=BCS_BoneSpace
    translationMode=BMM_Additive
    translationSpace=BCS_BoneSpace
    alphaBoolBlend {blendOption=Linear}
    alphaScaleBiasClamp {bInterpResult=true, interpSpeedDecreasing=2, interpSpeedIncreasing=2}
    boneToModify {boneName=dyn_sleeves_r_bk}
    rotation=<rot -10 15 -10>
    translation {x=-2}
   }
   n14 = "Transform (Modify) Bone - Bone: dyn_sleeves_r_bk"!(ComponentPose=@n13, bAlphaBoolEnabled:bool=True, Translation="0,0,0", Rotation="-20,25,0", Scale="1,1,1") -> (Pose) {
    alphaInputType=Bool
    rotationMode=BMM_Additive
    rotationSpace=BCS_BoneSpace
    translationMode=BMM_Additive
    translationSpace=BCS_BoneSpace
    alphaBoolBlend {blendOption=Linear}
    alphaScaleBiasClamp {bInterpResult=true, interpSpeedDecreasing=2, interpSpeedIncreasing=2}
    boneToModify {boneName=dyn_sleeves_r_bk}
    rotation {pitch=-20, yaw=25}
   }
   n15 = "Transform (Modify) Bone - Bone: dyn_sleeves_l_ft"!(ComponentPose=@n14, bAlphaBoolEnabled:bool=True, Translation="0,0,0", Rotation="0,25,0", Scale="1,1,1") -> (Pose) {
    alphaInputType=Bool
    rotationMode=BMM_Additive
    rotationSpace=BCS_BoneSpace
    translationMode=BMM_Additive
    translationSpace=BCS_BoneSpace
    alphaBoolBlend {blendOption=Linear}
    alphaScaleBiasClamp {bInterpResult=true, interpSpeedDecreasing=2, interpSpeedIncreasing=2}
    boneToModify {boneName=dyn_sleeves_l_ft}
    rotation {yaw=25}
   }
   n16 = Anim Dynamics - Bone: dyn_sleeves_r_MSTR!(ComponentPose=@n15, Alpha=1, GravityScale:float=1, ExternalForce="0,0,0") -> (Pose) {
    angularDampingOverride=1
    angularSpringConstant=500
    bGravityOverrideInSimSpace=true
    bOverrideAngularDamping=true
    bUseGravityOverride=true
    bUseSphericalLimits=true
    simulationSpace=BoneRelative
    alphaBoolBlend {blendOption=Linear}
    boundBone {boneName=dyn_sleeves_r_MSTR}
    componentAppliedLinearAccClamp=<3000 3000 3000>
    componentLinearAccScale {y=0.05}
    componentLinearVelScale {y=0.1}
    gravityOverride=<300 500 500>
    physicsBodyDefinitions [{
      sphereCollisionRadius=0
      boundBone {boneName=dyn_sleeves_r_MSTR}
      boxExtents=<15 15 15>
      constraintSetup {angularLimitsMax=<0 2 8>, angularLimitsMin=<0 -2 -5>, angularTarget=<0 0 0>}
      localJointOffset=<5 0 2>
     }]
    relativeSpaceBone {boneName=lowerarm_r}
    sphericalLimits [
     {limitRadius=3.5, drivingBone {boneName=lowerarm_twist_01_r}, sphereLocalOffset=<0 0 0>}
     {limitRadius=5, drivingBone {boneName=hand_r}, sphereLocalOffset=<5 0 0>}
    ]
   }
   n17 = Anim Dynamics - Bone: dyn_sleeves_r_up!(ComponentPose=@n16, Alpha=1, GravityScale:float=1, ExternalForce="0,0,0") -> (Pose) {
    angularDampingOverride=0.999
    angularSpringConstant=500
    bGravityOverrideInSimSpace=true
    bOverrideAngularDamping=true
    bUseGravityOverride=true
    bUseSphericalLimits=true
    simulationSpace=BoneRelative
    alphaBoolBlend {blendOption=Linear}
    boundBone {boneName=dyn_sleeves_r_up}
    componentAppliedLinearAccClamp=<3000 3000 3000>
    componentLinearAccScale=<7 7 7>
    componentLinearVelScale=<0.2 0.2 0.2>
    gravityOverride {y=600}
    physicsBodyDefinitions [{
      sphereCollisionRadius=0
      boundBone {boneName=dyn_sleeves_r_up}
      boxExtents=<15 15 15>
      constraintSetup {angularLimitsMax=<0 0 -5>, angularLimitsMin=<0 0 -10>, angularTarget=<0 0 0>}
      localJointOffset=<4 0 0>
     }]
    relativeSpaceBone {boneName=lowerarm_r}
    sphericalLimits [
     {limitRadius=4.4715, drivingBone {boneName=lowerarm_twist_01_r}, sphereLocalOffset=<0 0 0>}
     {limitRadius=5.2818, drivingBone {boneName=hand_r}, sphereLocalOffset=<-2.6414 0 0>}
    ]
   }
   n18 = Anim Dynamics - Bone: dyn_sleeves_r_dn!(ComponentPose=@n17, Alpha=1, GravityScale:float=1, ExternalForce="0,0,0") -> (Pose) {
    angularDampingOverride=0.99
    angularSpringConstant=1000
    bAngularSpring=true
    bGravityOverrideInSimSpace=true
    bOverrideAngularDamping=true
    bUseGravityOverride=true
    bUseSphericalLimits=true
    simulationSpace=RootRelative
    alphaBoolBlend {blendOption=Linear}
    boundBone {boneName=dyn_sleeves_r_dn}
    componentAppliedLinearAccClamp=<3000 3000 3000>
    componentLinearAccScale=<7 7 7>
    componentLinearVelScale=<0.3 0.3 0.3>
    gravityOverride {z=-800}
    physicsBodyDefinitions [{
      sphereCollisionRadius=0
      boundBone {boneName=dyn_sleeves_r_dn}
      boxExtents=<15 15 15>
      constraintSetup {angularLimitsMax=<0 0 8>, angularLimitsMin=<0 0 -5>, angularTarget=<0 0 0>}
      localJointOffset=<5 0 2>
     }]
    relativeSpaceBone {boneName=lowerarm_r}
    sphericalLimits [
     {limitRadius=3.5, drivingBone {boneName=lowerarm_twist_01_r}, sphereLocalOffset=<0 0 0>}
     {limitRadius=5, drivingBone {boneName=hand_r}, sphereLocalOffset=<5 0 0>}
     {limitRadius=5, drivingBone {boneName=pinky_metacarpal_r}, sphereLocalOffset=<0 0 3.557>}
    ]
   }
   n19 = Anim Dynamics - Bone: dyn_sleeves_r_ft!(ComponentPose=@n18, Alpha=1, GravityScale:float=1, ExternalForce="0,0,0") -> (Pose) {
    angularDampingOverride=1
    angularSpringConstant=500
    bGravityOverrideInSimSpace=true
    bOverrideAngularDamping=true
    bUseGravityOverride=true
    simulationSpace=BoneRelative
    alphaBoolBlend {blendOption=Linear}
    boundBone {boneName=dyn_sleeves_r_ft}
    componentAppliedLinearAccClamp=<3000 3000 3000>
    componentLinearAccScale=<7 7 7>
    componentLinearVelScale=<0.3 0.3 0.3>
    gravityOverride=<-500 300>
    physicsBodyDefinitions [{
      sphereCollisionRadius=0
      boundBone {boneName=dyn_sleeves_r_ft}
      boxExtents=<15 15 15>
      constraintSetup {angularLimitsMax=<0 2 10>, angularLimitsMin=<0 -2 2>, angularTarget=<0 0 0>}
      localJointOffset=<5 0 2>
     }]
    relativeSpaceBone {boneName=lowerarm_r}
    sphericalLimits [
     {limitRadius=3.9455, drivingBone {boneName=lowerarm_twist_01_r}, sphereLocalOffset=<0 0 0>}
     {limitRadius=5, drivingBone {boneName=hand_r}, sphereLocalOffset=<-1.8659 0 0>}
    ]
   }
   n20 = Anim Dynamics - Bone: dyn_sleeves_r_bk!(ComponentPose=@n19, Alpha=1, GravityScale:float=1, ExternalForce="0,0,0") -> (Pose) {
    angularDampingOverride=1
    angularSpringConstant=500
    bGravityOverrideInSimSpace=true
    bOverrideAngularDamping=true
    bUseGravityOverride=true
    simulationSpace=BoneRelative
    alphaBoolBlend {blendOption=Linear}
    boundBone {boneName=dyn_sleeves_r_bk}
    componentAppliedLinearAccClamp=<3000 3000 3000>
    componentLinearAccScale=<7 7 7>
    componentLinearVelScale=<0.3 0.3 0.3>
    gravityOverride=<-500 300>
    physicsBodyDefinitions [{
      sphereCollisionRadius=0
      boundBone {boneName=dyn_sleeves_r_bk}
      boxExtents=<15 15 15>
      constraintSetup {angularLimitsMax=<0 2 10>, angularLimitsMin=<0 -2 7>, angularTarget=<0 0 0>}
      localJointOffset=<5 -1 2>
     }]
    relativeSpaceBone {boneName=lowerarm_r}
    sphericalLimits [
     {limitRadius=4.6124, drivingBone {boneName=lowerarm_twist_01_r}, sphereLocalOffset=<0 0 0>}
     {limitRadius=4.5414, drivingBone {boneName=hand_r}, sphereLocalOffset=<8.5433 0 -0.7364>}
     {limitRadius=3.7026, drivingBone {boneName=pinky_metacarpal_r}, sphereLocalOffset=<0 0 1.24>}
     {limitRadius=3.6533, drivingBone {boneName=thumb_01_r}, sphereLocalOffset=<-0.9922 1.3345 0>}
    ]
   }
   n21 = Anim Dynamics - Bone: dyn_sleeves_l_MSTR!(ComponentPose=@n20, Alpha=1, GravityScale:float=1, ExternalForce="0,0,0") -> (Pose) {
    angularDampingOverride=1
    angularSpringConstant=500
    bGravityOverrideInSimSpace=true
    bOverrideAngularDamping=true
    bUseGravityOverride=true
    bUseSphericalLimits=true
    simulationSpace=BoneRelative
    alphaBoolBlend {blendOption=Linear}
    boundBone {boneName=dyn_sleeves_l_MSTR}
    componentAppliedLinearAccClamp=<3000 3000 3000>
    componentLinearAccScale {y=0.05}
    componentLinearVelScale {y=0.1}
    gravityOverride=<-300 -500 -300>
    physicsBodyDefinitions [{
      sphereCollisionRadius=0
      boundBone {boneName=dyn_sleeves_l_MSTR}
      boxExtents=<15 15 15>
      constraintSetup {angularLimitsMax=<0 2 8>, angularLimitsMin=<0 -2 -5>, angularTarget=<0 0 0>}
      localJointOffset=<5 0 0>
     }]
    relativeSpaceBone {boneName=lowerarm_l}
    sphericalLimits [
     {limitRadius=3.5, drivingBone {boneName=lowerarm_twist_01_l}, sphereLocalOffset=<0 0 0>}
     {limitRadius=5, drivingBone {boneName=hand_l}, sphereLocalOffset=<5 0 0>}
    ]
   }
   n22 = Anim Dynamics - Bone: dyn_sleeves_l_up!(ComponentPose=@n21, Alpha=1, GravityScale:float=1, ExternalForce="0,0,0") -> (Pose) {
    angularDampingOverride=0.999
    angularSpringConstant=500
    bGravityOverrideInSimSpace=true
    bOverrideAngularDamping=true
    bUseGravityOverride=true
    bUseSphericalLimits=true
    simulationSpace=BoneRelative
    alphaBoolBlend {blendOption=Linear}
    boundBone {boneName=dyn_sleeves_l_up}
    componentAppliedLinearAccClamp=<3000 3000 3000>
    componentLinearAccScale=<7 7 7>
    componentLinearVelScale=<0.2 0.2 0.2>
    gravityOverride {y=-500, z=100}
    physicsBodyDefinitions [{
      sphereCollisionRadius=0
      boundBone {boneName=dyn_sleeves_l_up}
      boxExtents=<15 15 15>
      constraintSetup {angularLimitsMax=<0 2 -5>, angularLimitsMin=<0 -2 -10>, angularTarget=<0 0 0>}
      localJointOffset=<5 0 0>
     }]
    relativeSpaceBone {boneName=lowerarm_l}
    sphericalLimits [
     {limitRadius=4.8452, drivingBone {boneName=lowerarm_twist_01_l}, sphereLocalOffset=<-1.9984 0 0>}
     {limitRadius=5, drivingBone {boneName=hand_l}, sphereLocalOffset=<2.6281 0 0>}
     {limitRadius=3, drivingBone {boneName=thumb_01_l}, sphereLocalOffset=<0 0 0>}
    ]
   }
   n23 = Anim Dynamics - Bone: dyn_sleeves_l_dn!(ComponentPose=@n22, Alpha=1, GravityScale:float=1, ExternalForce="0,0,0") -> (Pose) {
    angularDampingOverride=0.99
    angularSpringConstant=1000
    bAngularSpring=true
    bGravityOverrideInSimSpace=true
    bOverrideAngularDamping=true
    bUseGravityOverride=true
    bUseSphericalLimits=true
    simulationSpace=RootRelative
    alphaBoolBlend {blendOption=Linear}
    boundBone {boneName=dyn_sleeves_l_dn}
    componentAppliedLinearAccClamp=<3000 3000 3000>
    componentLinearAccScale=<7 7 7>
    componentLinearVelScale=<0.3 0.3 0.3>
    gravityOverride {z=-500}
    physicsBodyDefinitions [{
      sphereCollisionRadius=0
      boundBone {boneName=dyn_sleeves_l_dn}
      boxExtents=<15 15 15>
      constraintSetup {angularLimitsMax=<0 2 8>, angularLimitsMin=<0 -2 -5>, angularTarget=<0 0 0>}
      localJointOffset=<5 0 2>
     }]
    relativeSpaceBone {boneName=lowerarm_r}
    sphericalLimits [
     {limitRadius=3.5, drivingBone {boneName=lowerarm_twist_01_l}, sphereLocalOffset=<0 0 0>}
     {limitRadius=5, drivingBone {boneName=hand_l}, sphereLocalOffset=<5 0 0>}
     {limitRadius=3, drivingBone {boneName=thumb_01_l}, sphereLocalOffset=<0.8043 -0.451 0>}
    ]
   }
   n24 = Anim Dynamics - Bone: dyn_sleeves_l_ft!(ComponentPose=@n23, Alpha=1, GravityScale:float=1, ExternalForce="0,0,0") -> (Pose) {
    angularDampingOverride=1
    angularSpringConstant=500
    bGravityOverrideInSimSpace=true
    bOverrideAngularDamping=true
    bUseGravityOverride=true
    bUseSphericalLimits=true
    simulationSpace=BoneRelative
    alphaBoolBlend {blendOption=Linear}
    boundBone {boneName=dyn_sleeves_l_ft}
    componentAppliedLinearAccClamp=<3000 3000 3000>
    componentLinearAccScale=<7 7 7>
    componentLinearVelScale=<0.3 0.3 0.3>
    gravityOverride=<-500 -300>
    physicsBodyDefinitions [{
      sphereCollisionRadius=0
      boundBone {boneName=dyn_sleeves_l_ft}
      boxExtents=<15 15 15>
      constraintSetup {angularLimitsMax=<0 2 10>, angularLimitsMin=<0 -2 2>, angularTarget=<0 0 0>}
      localJointOffset=<5 1 2>
     }]
    relativeSpaceBone {boneName=lowerarm_l}
    sphericalLimits [
     {limitRadius=3.5, drivingBone {boneName=lowerarm_twist_01_l}, sphereLocalOffset=<0 0 0>}
     {limitRadius=5, drivingBone {boneName=hand_l}, sphereLocalOffset=<5 0 0>}
     {limitRadius=3, drivingBone {boneName=thumb_01_l}, sphereLocalOffset=<1.3096 0 0>}
    ]
   }
   n25 = Anim Dynamics - Bone: dyn_sleeves_l_bk!(ComponentPose=@n24, Alpha=1, GravityScale:float=1, ExternalForce="0,0,0") -> (Pose) {
    angularDampingOverride=1
    angularSpringConstant=500
    bGravityOverrideInSimSpace=true
    bOverrideAngularDamping=true
    bUseGravityOverride=true
    bUseSphericalLimits=true
    simulationSpace=BoneRelative
    alphaBoolBlend {blendOption=Linear}
    boundBone {boneName=dyn_sleeves_l_bk}
    componentAppliedLinearAccClamp=<3000 3000 3000>
    componentLinearAccScale=<7 7 7>
    componentLinearVelScale=<0.3 0.3 0.3>
    gravityOverride=<-500 -300>
    physicsBodyDefinitions [{
      sphereCollisionRadius=0
      boundBone {boneName=dyn_sleeves_l_bk}
      boxExtents=<15 15 15>
      constraintSetup {angularLimitsMax=<0 2 10>, angularLimitsMin=<0 -2 7>, angularTarget=<0 0 0>}
      localJointOffset=<5 -1 2>
     }]
    relativeSpaceBone {boneName=lowerarm_l}
    sphericalLimits [
     {limitRadius=3.5, drivingBone {boneName=lowerarm_twist_01_l}, sphereLocalOffset=<0 0 0>}
     {limitRadius=5, drivingBone {boneName=hand_l}, sphereLocalOffset=<5 0 0>}
     {limitRadius=2, drivingBone {boneName=thumb_01_l}, sphereLocalOffset=<0 0 0>}
    ]
   }
   Output Pose(Result=Component To Local(ComponentPose=@n25)) {layerGroup=None}
   Save cached pose 'SavedPose26139'(Pose=Copy Pose From Mesh() {bUseAttachedParent=true}) {cacheName=SavedPose26139}
  ]}
 Disable_WPO {nodes [
   Disable_WPO() {is=title K2Node_FunctionEntry}
   > n1 = Sequence()
   n2 = Get Mat Instance Dynamic()
   > SetScalarParameterValue(self=@n2, ParameterName=ClothSim_R1_DragStrength, Value=0)
   > SetScalarParameterValue(self=@n2, ParameterName=ClothSim_R1_TurbulenceStrength, Value=0)
   > SetScalarParameterValue(self=@n2, ParameterName=ClothSim_R1_BoneRadius, Value=0)
   > SetScalarParameterValue(self=@n2, ParameterName=ClothSim_R1_BoneThreshold, Value=0)
   SetScalarParameterValue(execute=@n1.then_1, self=@n2, ParameterName=ClothSim_R2_DragStrength, Value=0)
   > SetScalarParameterValue(self=@n2, ParameterName=ClothSim_R2_TurbulenceStrength, Value=0)
   > SetScalarParameterValue(self=@n2, ParameterName=ClothSim_R2_BoneRadius, Value=0)
   > SetScalarParameterValue(self=@n2, ParameterName=ClothSim_R2_BoneThreshold, Value=0)
  ]}
 Enable_WPO {nodes [
   Enable_WPO() {is=title K2Node_FunctionEntry}
   > n1 = Sequence()
   n2 = Get Mat Instance Dynamic()
   n3 = Get!() -> (WPO_R1_Default_X:double, WPO_R1_Default_Y:double, WPO_R1_Default_Z:double, WPO_R1_Default_W:double)
   > SetScalarParameterValue(self=@n2, ParameterName=ClothSim_R1_DragStrength, Value=@n3)
   > SetScalarParameterValue(self=@n2, ParameterName=ClothSim_R1_TurbulenceStrength, Value=@n3.WPO_R1_Default_Y)
   > SetScalarParameterValue(self=@n2, ParameterName=ClothSim_R1_BoneRadius, Value=@n3.WPO_R1_Default_Z)
   > SetScalarParameterValue(self=@n2, ParameterName=ClothSim_R1_BoneThreshold, Value=@n3.WPO_R1_Default_W)
   n4 = Get!() -> (WPO_R2_Default_X:double, WPO_R2_Default_Y:double, WPO_R2_Default_Z:double, WPO_R2_Default_W:double)
   SetScalarParameterValue(execute=@n1.then_1, self=@n2, ParameterName=ClothSim_R2_DragStrength, Value=@n4)
   > SetScalarParameterValue(self=@n2, ParameterName=ClothSim_R2_TurbulenceStrength, Value=@n4.WPO_R2_Default_Y)
   > SetScalarParameterValue(self=@n2, ParameterName=ClothSim_R2_BoneRadius, Value=@n4.WPO_R2_Default_Z)
   > SetScalarParameterValue(self=@n2, ParameterName=ClothSim_R2_BoneThreshold, Value=@n4.WPO_R2_Default_W)
  ]}
 EventGraph {nodes [
   Event OnCharacterCustomizationLoaded()
   > n1 = Set SkelMesh(SkelMesh=GetPartSkeletalMeshComponent())
   > n2 = Get Dynamic Material Instance(SkeletalMeshComponent=@n1)
   > Set Mat Instance Dynamic(Mat Instance Dynamic=@n2)
   > n3 = GetOrCreateDynamicOverlayMaterial(MeshComponent=@n1, MaterialName=None)
   > Set Overlay Mat Dynamic(Overlay Mat Dynamic:Material Instance Dynamic Object Reference=@n3) -> (Output_Get:Material Instance Dynamic Object Reference)
   > n4 = Set is_Wearing_Backpack(is_Wearing_Backpack:bool=CheckForBackbling()) -> (Output_Get:bool)
   > n5 = Branch(Condition=@n4)
   > Set Alpha(Alpha:float=0) -> (Output_Get:float)
   > Set Alpha_0(Alpha_0:float=0) -> (Output_Get:float)
   > Set Alpha_1(Alpha_1:float=0) -> (Output_Get:float)
   Set Alpha(execute=@n5.else, Alpha=1)
   > Set Alpha_0(Alpha_0=1)
   > Set Alpha_1(Alpha_1=1)
   Event BlueprintInitializeAnimation()
   > Set SkelMesh(SkelMesh=GetOwnerSkeletalMeshComponent())
   n6 = TryGetPawnOwner()
   > Is Valid(InputObject=@n6)
   > n7 = Cast To FortPlayerPawnAthena(Object=@n6)
   > Set Player Pawn(Player Pawn:Fort Player Pawn Athena Object Reference=@n7) -> (Output_Get:Fort Player Pawn Athena Object Reference)
   n8 = Event BlueprintUpdateAnimation()
   > n9 = Set Delta Time X(Delta Time X=@n8.DeltaTimeX)
   > n10 = Sequence()
   > Is Valid(InputObject=Get Player Pawn() -> (Player Pawn:Fort Player Pawn Athena Object Reference))
   > Set WPO Globals Function(
    Material=Get Mat Instance Dynamic()
    Overlay Material=Get Overlay Mat Dynamic() -> (Overlay Mat Dynamic:Material Instance Dynamic Object Reference)
    Skeletal Mesh=Get SkelMesh()
    "Local Noise Transform  Location"="Get Local Noise Transform  Location"() -> ("Local Noise Transform  Location":Vector)
    Local Noise Offset=Get Local Noise Offset()
   )
   > Set WPO Regional Function!(
    execute
    Material:Material Instance Dynamic Object Reference=Get Mat Instance Dynamic()
    Overlay Material:Material Instance Dynamic Object Reference=Get Overlay Mat Dynamic()
    Skeletal Mesh:Skeletal Mesh Component Object Reference=Get SkelMesh()
    Velocity Transform:Name=foot_r
    Velocity Interp Speed:double=Get Velocity Interp Speed()
    Collision Transform:Name=calf_r
    Collision Transform Axis:EAxis Enum
    Invert Collision Transform Axis:bool=false
    Delta Time X:double=@n9
    Previous Velocity Transform Location:"Vector (by ref)"=Get R2 Velocity Transform Location()
    Previous Velocity Transform Velocity:"Vector (by ref)"=Get R2 Velocity Transform Velocity()
    Region:int=2
   ) -> (then)
   > Set WPO Regional Function(
    Material=Get Mat Instance Dynamic()
    Overlay Material=Get Overlay Mat Dynamic()
    Skeletal Mesh=Get SkelMesh()
    Velocity Transform=hand_r
    Velocity Interp Speed=Get Velocity Interp Speed()
    Collision Transform=lowearm_r
    Invert Collision Transform Axis=false
    Delta Time X=@n9
    Previous Velocity Transform Location=Get R1 Velocity Transform Location()
    Previous Velocity Transform Velocity=Get R1 Velocity Transform Velocity()
    Region=1
   )
   > SetScalarParameterValue(self=Get Mat Instance Dynamic(), ParameterName=ClothSim_WindStrength, Value=SelectFloat(A=1, B=0, bPickA=Get bIsSkydiveFloating()))
   n11 = Get SkelMesh()
   Set R_sleeve_fix(
    execute=@n10.then_1
    R_sleeve_fix:double=SelectFloat(
     A=GetAlphaFromReferenceTransformDelta(SkeletalMeshComponent=@n11, BoneName=hand_r, BaseName=lowearm_r, InputRange=(X=-26,Y=154), TwistAxis="1,0,0")
     B=0
     bPickA=Get bIsTargeting()
    )
   ) -> (Output_Get:double)
   TryGetPawnOwner()
  ]}
 NewEventGraph {links [], nodes []}
 Save WPO Defaults {nodes [
   Save WPO Defaults() {is=title K2Node_FunctionEntry}
   > n1 = Sequence!(execute) -> (then_0, then_1, then_2:Exec)
   n2 = Get Mat Instance Dynamic()
   > n3 = Get Scalar Parameter Value!(execute, self:Material Instance Dynamic Object Reference=@n2, ParameterName:Name=ClothSim_R1_DragStrength) -> (then, ReturnValue:float) {is=title K2Node_CallFunction}
   n4 = Get Mat Instance Dynamic()
   > n5 = Get Scalar Parameter Value(self=@n4, ParameterName=ClothSim_R1_TurbulenceStrength)
   n6 = Get Mat Instance Dynamic()
   > n7 = Get Scalar Parameter Value(self=@n6, ParameterName=ClothSim_R1_BoneRadius)
   n8 = Get Mat Instance Dynamic()
   > n9 = Get Scalar Parameter Value(self=@n8, ParameterName=ClothSim_R1_BoneThreshold)
   > Set!(execute, WPO_R1_Default_X:double=@n3, WPO_R1_Default_Y:double=@n5, WPO_R1_Default_Z:double=@n7, WPO_R1_Default_W:double=@n9) -> (then, Output_Get:Vector 4 Structure)
   n10 = Get Scalar Parameter Value(execute=@n1.then_1, self=@n2, ParameterName=ClothSim_R2_DragStrength)
   > n11 = Get Scalar Parameter Value(self=@n4, ParameterName=ClothSim_R2_TurbulenceStrength)
   > n12 = Get Scalar Parameter Value(self=@n6, ParameterName=ClothSim_R2_BoneRadius)
   > n13 = Get Scalar Parameter Value(self=@n8, ParameterName=ClothSim_R2_BoneThreshold)
   > Set!(execute, WPO_R2_Default_X:double=@n10, WPO_R2_Default_Y:double=@n11, WPO_R2_Default_Z:double=@n12, WPO_R2_Default_W:double=@n13) -> (then, Output_Get)
  ]}
]
variables [
 Alpha {type=number, default=1}
 Alpha_0 {type=number, default=1}
 Alpha_1 {type=number, default=1}
 AlphaRotationHistory {type=number}
 As Player Pawn Athena {type=PlayerPawn_Athena_C}
 Delta Time X {type=number}
 Delta Time X_0 {type=number}
 History {type=PositionHistory, default {range=0, positions []}}
 is_Wearing_Backpack {type=boolean}
 Local Noise Offset {type=Vector}
 Local Noise Offset_0 {type=Vector}
 "Local Noise Transform  Location" {type=Vector}
 Local Noise Transform Location {type=Vector}
 Mat Instance Dynamic {type=MaterialInstanceDynamic}
 Overlay Mat Dynamic {type=MaterialInstanceDynamic}
 Overlay Mat Inst Dynamic {type=MaterialInstanceDynamic}
 Player Pawn {type=FortPlayerPawnAthena}
 R1 Velocity Transform Location {type=Vector}
 R1 Velocity Transform Velocity {type=Vector}
 R2 Velocity Transform Location {type=Vector}
 R2 Velocity Transform Velocity {type=Vector}
 R_sleeve_fix {type=number, default=1}
 SkelMesh {type=SkeletalMeshComponent}
 Spine_05_Rot {type=Vector}
 Velocity Interp Speed {type=number, default=10}
 Velocity Interp Speed_0 {type=number, default=1}
 WPO Pho Wind Intensity {type=number}
 WPO_R1_Default {type=Vector4, default=<0 0 0 0>}
 WPO_R2_Default {type=Vector4, default=<0 0 0 0>}
]
